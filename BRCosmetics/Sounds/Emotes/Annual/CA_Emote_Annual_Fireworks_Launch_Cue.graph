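# Sound cue for the firework launch sound in the Annual fireworks emote.
# A Random node picks one of twelve equally weighted launch wave variations, and a Modulator adds slight pitch and volume variation.
# It uses emote attenuation, the emote music sound class, and emote foley concurrency, and sends to a music audio-analysis submix.
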SoundCue /BRCosmetics/Sounds/Emotes/Annual/CA_Emote_Annual_Fireworks_Launch_Cue
root=Modulator_0
nodes [
 {id=Modulator_0, children [Random_1], properties {pitchMax=1.1, pitchMin=0.9, volumeMax=1, volumeMin=0.8}}
 {
  id=Random_1
  children [
   WavePlayer_1
   WavePlayer_4
   WavePlayer_11
   WavePlayer_12
   WavePlayer_13
   WavePlayer_14
   WavePlayer_15
   WavePlayer_16
   WavePlayer_17
   WavePlayer_18
   WavePlayer_0
   WavePlayer_2
  ]
  properties {weights [1,1,1,1,1,1,1,1,1,1,1,1]}
 }
 {id=WavePlayer_0, properties {soundWaveAssetPtr=&/Emote_Annual_Firework_Launch_11}}
 {id=WavePlayer_1, properties {soundWaveAssetPtr=&/Emote_Annual_Firework_Launch_01}}
 {id=WavePlayer_11, properties {soundWaveAssetPtr=&/Emote_Annual_Firework_Launch_03}}
 {id=WavePlayer_12, properties {soundWaveAssetPtr=&/Emote_Annual_Firework_Launch_04}}
 {id=WavePlayer_13, properties {soundWaveAssetPtr=&/Emote_Annual_Firework_Launch_05}}
 {id=WavePlayer_14, properties {soundWaveAssetPtr=&/Emote_Annual_Firework_Launch_06}}
 {id=WavePlayer_15, properties {soundWaveAssetPtr=&/Emote_Annual_Firework_Launch_07}}
 {id=WavePlayer_16, properties {soundWaveAssetPtr=&/Emote_Annual_Firework_Launch_08}}
 {id=WavePlayer_17, properties {soundWaveAssetPtr=&/Emote_Annual_Firework_Launch_09}}
 {id=WavePlayer_18, properties {soundWaveAssetPtr=&/Emote_Annual_Firework_Launch_10}}
 {id=WavePlayer_2, properties {soundWaveAssetPtr=&/Emote_Annual_Firework_Launch_12}}
 {id=WavePlayer_4, properties {soundWaveAssetPtr=&/Emote_Annual_Firework_Launch_02}}
]
properties {
 attenuationSettings=&Fort_Emotes_Athena_Att
 maxDistance=1600
 soundClassObject=&Fort_Music_Emote
 volumeMultiplier=1
 concurrencyOverrides {maxCount=3}
 concurrencySet [/Game/Sounds/Concurrency_Rules/Athena_EmoteFoley_Conc]
 soundSubmixSends [{minSendLevel=0, soundSubmix=&Fortnite_AudioAnalysis_Music}]
}
