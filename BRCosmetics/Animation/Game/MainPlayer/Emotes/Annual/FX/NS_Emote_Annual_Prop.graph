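# Niagara effect for the Annual emote prop, drawing translucent pale-blue speed-blur meshes.
# Two emitters each burst-spawn mesh particles at the system location with fixed color, scale, rotation and lifetime.
# SpeedBlur is larger and fainter, using two mesh renderers, stacked orientation modules and a slight downward offset; SpeedBlur001 is a smaller, more opaque single mesh.
# Particle update only handles lifetime, with no velocity or scaling, and the system exposes no user parameters.
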NiagaraSystem /BRCosmetics/Animation/Game/MainPlayer/Emotes/Annual/FX/NS_Emote_Annual_Prop
dependencies {usedDataInterfaces [], usedDynamicInputs [], usedModules [&SystemState]}
emitters {
 SpeedBlur {
  emitterUpdateScript [
   &EmitterState(Inactive Response=::NewEnumerator1, Life Cycle Mode=::NewEnumerator1, Loop Behavior=::NewEnumerator1, Loop Duration=2, UseLoopDelay=false)
   &SpawnBurst_Instantaneous(Use Spawn Probability=false, UseLoopCountLimit=false)
  ]
  particleSpawnScript [
   &InitializeParticle(
    Color=<rgba 0.5955 0.7673 1 0.1>
    Lifetime=5
    Mesh Scale=<0.7 0.5 0.7>
    Mesh Scale Mode=::NewEnumerator4
    UsePositionOffset=false
    Write Material Random=true
   )
   &InitialMeshOrientation(Mesh Orientation Mode=::NewEnumerator7, Rotation=<0.25 0 0>, Use Rotation Vector=true)
   &InitialMeshOrientation(Mesh Orientation Mode=::NewEnumerator7, Rotation=<0 0 0.34>, Use Rotation Vector=true) {module=InitialMeshOrientation001}
   &SystemLocation(MaskSpawnGroup=false, Offset=<0 0 -5>)
  ]
  particleUpdateScript [&ParticleState()]
  rendererClasses [&NiagaraMeshRendererProperties]
  renderers [
   {rendererClass=&NiagaraMeshRendererProperties, rendererIndex=0}
   {rendererClass=&NiagaraMeshRendererProperties, rendererIndex=1}
  ]
 }
 SpeedBlur001 {
  emitterUpdateScript [
   &EmitterState(Inactive Response=::NewEnumerator1, Life Cycle Mode=::NewEnumerator1, Loop Behavior=::NewEnumerator1, Loop Duration=2, UseLoopDelay=false)
   &SpawnBurst_Instantaneous(Use Spawn Probability=false, UseLoopCountLimit=false)
  ]
  particleSpawnScript [
   &InitializeParticle(
    Color=<rgba 0.6311 0.7242 0.8594 0.5>
    Lifetime=5
    Mesh Scale=<0.07 0.07 0.05>
    Mesh Scale Mode=::NewEnumerator4
    UsePositionOffset=false
    Write Material Random=true
   )
   &InitialMeshOrientation(Mesh Orientation Mode=::NewEnumerator7, Rotation=<0 0 -0.25>, Use Rotation Vector=true)
   &SystemLocation(MaskSpawnGroup=false)
  ]
  particleUpdateScript [&ParticleState()]
  rendererClasses [&NiagaraMeshRendererProperties]
  renderers [{rendererClass=&NiagaraMeshRendererProperties, rendererIndex=0}]
 }
}
summary {
 systemName=NS_Emote_Annual_Prop
 emitters [
  {emitterName=SpeedBlur, rendererClasses [&NiagaraMeshRendererProperties]}
  {emitterName=SpeedBlur001, rendererClasses [&NiagaraMeshRendererProperties]}
 ]
 userVariables []
}
